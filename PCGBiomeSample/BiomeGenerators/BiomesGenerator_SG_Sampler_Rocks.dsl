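; PCG sampler subgraph for a biome generator that produces rock spawn points on the landscape.
; Input data tagged BoundingShape bounds a Surface Sampler over Get Landscape Data, which skips layer weights and does not require components to overlap itself.
; The sampler spreads large-extent points loosely without applying density, then Transform Points adds full random yaw, slight pitch and roll tilt, and moderate scale variation before output.
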(graph BiomesGenerator_SG_Sampler_Rocks
  (bind FilterByTag_0 (Filter%20Data%20By%20Tag :bTokenizeOnWhiteSpace true :SelectedTags "BoundingShape"))
  (bind GetLandscapeData_0 (Get%20Landscape%20Data :bComponentsMustOverlapSelf false :SamplingProperties (struct :bGetLayerWeights false)))
  (bind SurfaceSampler_0 (Surface%20Sampler :bApplyDensityToPoints false :Looseness 3.0 :PointExtents (4e+02 4e+02 4e+02)))
  ;[0 0 0] Transforms
  (bind TransformPoints_1 (Transform%20Points :RotationMax (struct :Pitch 1e+01 :Yaw 1.8e+02 :Roll 1e+01) :RotationMin (struct :Pitch -1e+01 :Yaw -1.8e+02 :Roll -1e+01) :ScaleMax (1.25 1.25 1.25) :ScaleMin (0.75 0.75 0.75)))
  (connect DefaultInputNode In FilterByTag_0 In)
  (connect FilterByTag_0 InsideFilter SurfaceSampler_0 Bounding%20Shape)
  (connect GetLandscapeData_0 Out SurfaceSampler_0 Surface)
  (connect SurfaceSampler_0 Out TransformPoints_1 In)
  (output Out TransformPoints_1))
