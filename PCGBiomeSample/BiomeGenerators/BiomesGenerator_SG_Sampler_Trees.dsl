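; PCG sampler subgraph for biome tree generation that outputs randomized tree candidate points on the landscape.
; Surface sampling within input tagged BoundingShape uses parameterized spacing, extents and seed.
; A proxy runs swappable spatial noise settings, whose inverted density is capped by random noise and threshold-filtered into clusters.
; Surviving points get fresh density noise and random rotation and scale; a constant-density branch stays unconnected to the output.
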(graph BiomesGenerator_SG_Sampler_Trees
  (params (SpatialNoiseSetting :type (SoftObject /Script/PCG.PCGSettingsInterface) :default /PCGBiomeSample/BiomeGenerators/SpatialNoise_FracBrownian2D.SpatialNoise_FracBrownian2D) (SpatialNoiseThreshold :type Double :default 0.66) (PointsPerSquaredMeter :type Double :default 0.05) (PointsExtents :type Vector :default (1e+02 1e+02 1e+02)) (Looseness :type Double :default 1.25) (Seed :type Int :default -1393707107))
  (bind FilterByTag_1 (Filter%20Data%20By%20Tag :bTokenizeOnWhiteSpace true :SelectedTags "BoundingShape"))
  (bind GetLandscapeData_1 (Get%20Landscape%20Data :bComponentsMustOverlapSelf false :SamplingProperties (struct :bGetLayerWeights false)))
  (bind SurfaceSampler_0 (Surface%20Sampler :bApplyDensityToPoints false :Looseness 1.25 :PointExtents (1e+02 1e+02 5e+02) :PointsPerSquaredMeter 0.05 :Looseness-param Looseness :PointExtents-param PointsExtents :PointsPerSquaredMeter-param PointsPerSquaredMeter :Seed-param Seed))
  ;[0.00868 0.00868 0.00868 1 12] Proxy Noise
  (bind Proxy_0 (Proxy :bTagOutputsBasedOnOutputPins false :ProxyInterfaceMode ByNativeElement :SettingsClass /Script/PCG.PCGSpatialNoiseSettings :Settings-param SpatialNoiseSetting))
  (bind DensityRemap_5 (Density%20Remap :OutRangeMax 0.0 :OutRangeMin 1.0))
  (bind AttributeNoise_0 (Attribute%20Noise :Mode Minimum :NoiseMin 0.5))
  (bind DensityFilter_4 (Density%20Filter :LowerBound 0.0 :UpperBound-param SpatialNoiseThreshold))
  (bind CreateAttribute_13 (Create%20Constant :AttributeTypes (struct :DoubleValue 1.0)))
  (bind AttributeMathsOp_12 (Attribute%20Maths%20Op :InputSource1 $Density :Operation Set))
  (bind AttributeNoise_1 (Attribute%20Noise))
  ;[0 0 0] Transforms
  (bind TransformPoints_0 (Transform%20Points :RotationMax (struct :Pitch 1e+01 :Yaw 1.8e+02 :Roll 1e+01) :RotationMin (struct :Pitch -1e+01 :Yaw -1.8e+02 :Roll -1e+01) :ScaleMax (1.25 1.25 1.25) :ScaleMin (0.75 0.75 0.75)))
  (connect DensityFilter_4 Out AttributeMathsOp_12 InA)
  (connect CreateAttribute_13 Out AttributeMathsOp_12 InB)
  (connect DensityRemap_5 Out AttributeNoise_0 In)
  (connect DensityFilter_4 Out AttributeNoise_1 In)
  (connect AttributeNoise_0 Out DensityFilter_4 In)
  (connect Proxy_0 Out DensityRemap_5 In)
  (connect DefaultInputNode In FilterByTag_1 In)
  (connect SurfaceSampler_0 Out Proxy_0 In)
  (connect FilterByTag_1 InsideFilter SurfaceSampler_0 Bounding%20Shape)
  (connect GetLandscapeData_1 Out SurfaceSampler_0 Surface)
  (connect AttributeNoise_1 Out TransformPoints_0 In)
  (output Out TransformPoints_0))
